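# Frontend bottom bar variant of the UIKit Quiet button that adds chat notification counts and a mobile mic icon.
# SetUniqueButtonAttributes detects chat buttons by "Chat" action description or row name, then manages a lazily created status indicator in the Root grid.
# On touch, the mic action gets a talk or mute icon; other actions clear it.
# Input method changes make the button hit-test-invisible on gamepad, visible otherwise.
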WidgetBlueprint /Game/UI/Frontend/BottomBar/WBP_UIKit_Button_Quiet_BottomBar
graphs [
 AddStatusIndicator {nodes [
   AddStatusIndicator() {is=title K2Node_FunctionEntry}
   > RemoveWidget!(execute, Widget:Widget Object Reference=Get StatusIndicator() -> (StatusIndicator:WBP UIKit Status Indicator Object Reference)) -> (then)
   > n1 = Create Widget!(
    execute:Exec
    Class:User Widget Class Reference=Get StatusIndicatorClass() -> (StatusIndicatorClass:WBP UIKit Status Indicator Class Reference)
    OwningPlayer:Player Controller Object Reference
   ) -> (then:Exec, ReturnValue:WBP UIKit Status Indicator Object Reference)
   > n2 = Set StatusIndicator(StatusIndicator:WBP UIKit Status Indicator Object Reference=@n1) -> (Output_Get:WBP UIKit Status Indicator Object Reference)
   > Branch(Condition=IsValid(Object=@n2))
   > SetIndicatorType!(execute, self:WBP UIKit Status Indicator Base Object Reference=@n2, IndicatorType:E_UI_StatusIndicator_Type Enum=NewEnumerator2) -> (then)
   > SetIndicatorConfiguration!(
    execute
    self:WBP UIKit Status Indicator Base Object Reference=@n2
    InCount:int=0
    IndicatorConfig:E_UI_StatusIndicator_Configuration Enum=NewEnumerator3
   ) -> (then)
   > AddChildToGrid!(
    execute
    self:Grid Panel Object Reference=Get Root() -> (Root:Grid Panel Object Reference)
    Content:Widget Object Reference=@n2
    InRow:int=1
    InColumn:int=4
   ) -> (then, ReturnValue:Grid Slot Object Reference)
   > SetGridSlotLayer!(execute, Widget:Widget Object Reference=@n2, Layer:int=100) -> (then)
  ]}
 DetermineIsChatButton {nodes [
   DetermineIsChatButton() {is=title K2Node_FunctionEntry}
   n1 = Get InputActionWidget() -> (InputActionWidget:Common Action Widget Object Reference)
   n2 = GetEnhancedInputAction!(self:Common Action Widget Object Reference=@n1) -> (ReturnValue:Input Action Object Reference)
   > n3 = Branch(Condition=IsValid(Object=@n2))
   > n4 = Branch(Condition="Equal, Case Insensitive (Text)"!(
     A:"Text (by ref)"=Get ActionDescription!(self:Input Action Object Reference=@n2) -> (ActionDescription:Text)
     B:"Text (by ref)"=Format Text(Format="LOCTEXT(\"Chat\")")
    ) -> (ReturnValue:bool))
   > Return Node!(execute, IsChatButton:bool=true)
   n5 = For Each Loop(
    Exec=[@n4.else, @n3.else]
    Array:Array of Data Table Row Handle Structures=Get InputActions!(self:Common Action Widget Object Reference=@n1) -> (InputActions:Array of Data Table Row Handle Structures)
   ) -> (Array Element:Data Table Row Handle Structure)
   Return Node(execute=@n5.Completed, IsChatButton=false)
   n6 = Break Data Table Row Handle!(DataTableRowHandle:"Data Table Row Handle Structure (by ref)"=Get ChatInputActionGamepad() -> (ChatInputActionGamepad:Data Table Row Handle Structure)) -> (DataTable:Data Table Object Reference, RowName:Name) {is=title K2Node_BreakStruct}
   n7 = Break Data Table Row Handle(DataTableRowHandle=@n5)
   n8 = Break Data Table Row Handle(DataTableRowHandle=Get ChatInputActionPC() -> (ChatInputActionPC:Data Table Row Handle Structure))
   > Branch(Condition=OR Boolean(A="Equal (Name)"(A=@n7.RowName, B=@n8.RowName), B="Equal (Name)"(A=@n7.RowName, B=@n6.RowName)))
   > Return Node(IsChatButton=true)
  ]}
 DetermineIsMobileMicButton {nodes [
   DetermineIsMobileMicButton() {is=title K2Node_FunctionEntry}
   > n1 = For Each Loop(Array=Get InputActions(self=Get InputActionWidget()))
   n2 = Get MobileMicInputAction() -> (MobileMicInputAction:Data Table Row Handle Structure)
   > n3 = Branch(Condition="Equal (Name)"(A=Break Data Table Row Handle(DataTableRowHandle=@n1).RowName, B=Break Data Table Row Handle(DataTableRowHandle=@n2).RowName))
   > Return Node!(execute, IsMobileMicButton:bool=true)
   Return Node(execute=@n3.else, IsMobileMicButton=false)
  ]}
 EventGraph {nodes [
   n1 = Event PreConstruct()
   > Parent: PreConstruct(IsDesignTime=@n1.IsDesignTime)
   > Branch(Condition=@n1.IsDesignTime)
   > AddStatusIndicator() {is=title K2Node_CallFunction}
   n2 = Event On Input Method Changed!() -> (OutputDelegate, then, CurrentInputType:ECommonInputType Enum) {is=title K2Node_Event}
   > n3 = Parent: On Input Method Changed!(execute, CurrentInputType:ECommonInputType Enum=@n2.CurrentInputType) -> (then) {is=title K2Node_CallParentFunction}
   n4 = Get CurrentInputType() -> (CurrentInputType:ECommonInputType Enum)
   SetVisibility(
    execute=[@n5.then, @n3.then]
    InVisibility=Select(
     Option 0=HitTestInvisible
     Option 1=Visible
     Index="Not Equal (Enum)"!(A:ECommonInputType Enum=@n4, B:ECommonInputType Enum=Gamepad) -> (ReturnValue:bool)
    )
   )
   Event Construct()
   > n5 = Parent: Construct()
  ]}
 RemoveStatusIndicator {nodes [
   RemoveStatusIndicator() {is=title K2Node_FunctionEntry}
   n1 = Get StatusIndicator()
   > Branch(Condition=IsValid(Object=@n1))
   > RemoveFromParent(self=@n1)
   > Set StatusIndicator()
  ]}
 SetUniqueButtonAttributes {nodes [
   n1 = SetUniqueButtonAttributes!() -> (then, NotificationCount:int, IsMicOn:bool=~) {is=title K2Node_FunctionEntry}
   > n2 = Sequence()
   > n3 = DetermineIsChatButton!(execute, self) -> (then, IsChatButton:bool) {is=title K2Node_CallFunction}
   > n4 = Branch(Condition=@n3)
   > n5 = Branch(Condition=IsValid(Object=Get StatusIndicator()))
   n6 = Branch(execute=[@n8.then, @n5.then], Condition="Equal (Integer)"(A=@n1))
   n7 = Get StatusIndicator()
   > SetVisibility(self=@n7, InVisibility=Collapsed)
   OverrideCount!(execute=@n6.else, self:WBP UIKit Status Indicator Base Object Reference=@n7, Count:int=@n1) -> (then)
   > SetVisibility(self=@n7, InVisibility=HitTestInvisible)
   n8 = AddStatusIndicator(execute=@n5.else)
   RemoveStatusIndicator(execute=@n4.else) {is=title K2Node_CallFunction}
   Branch(
    execute=@n2.then_1
    Condition=Select!(MouseAndKeyboard:bool=false, Gamepad:bool=false, Touch:bool=true, Count:bool=false, Index:ECommonInputType Enum=Get CurrentInputType()) -> (ReturnValue:bool)
   )
   > n9 = DetermineIsMobileMicButton!(execute, self) -> (then, IsMobileMicButton:bool) {is=title K2Node_CallFunction}
   > n10 = Branch(Condition=@n9)
   > Set HasIcon(HasIcon:bool=true) -> (Output_Get:bool)
   > SetIconMaterial!(
    execute
    self:WBP UIKit Button CTA Base Object Reference
    IconMaterial:Material Instance Object Reference=/FortUILibrary/UIKit/Buttons/MaterialInstances/GenericButtons/MI_GenericIconDeepBlue
   ) -> (then)
   > SetIcon!(
    execute
    self:WBP UIKit Button CTA Base Object Reference
    Icon:Object Reference=Select!(
     Option 0:Object Reference=/Game/Athena/HUD/Art/BacchusHudIcons/Mute_Normal
     Option 1:Object Reference=/Game/Athena/HUD/Art/BacchusHudIcons/Talk_Normal
     Index:bool=@n1.IsMicOn
    ) -> (ReturnValue:Object Reference)
   ) -> (then)
   > UpdateButtonPadding(self:WBP UIKit Button CTA Base Object Reference)
   Set HasIcon(execute=@n10.else, HasIcon=false)
   > SetIconMaterial()
   > SetIcon()
  ]}
]
variables [
 ChatInputActionGamepad {type=DataTableRowHandle, default {dataTable=&InputActionDataTable, rowName=ChatV2}}
 ChatInputActionPC {type=DataTableRowHandle, default {dataTable=&InputActionDataTable, rowName=ChatOnlyPC}}
 MobileMicInputAction {type=DataTableRowHandle, default {dataTable=&InputActionDataTable, rowName=MobileMicToggle}}
 StatusIndicator {type=WBP_UIKit_StatusIndicator_C, default=null}
 StatusIndicatorClass {type=Class, default=/FortUILibrary/UIKit/StatusIndicator/WBP_UIKit_StatusIndicator.WBP_UIKit_StatusIndicator_C}
]
widgets {
 info {
  inheritedWidgetCount=0
  parentClass=/FortUILibrary/UIKit/Buttons/WBP_UIKit_Button_Quiet.WBP_UIKit_Button_Quiet_C
  rootWidgetClass=null
  widgetCount=0
 }
 widgets []
}
