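# AthenaHUDBoundActionButton is a FortBoundActionButton widget for a frontend bottom bar, showing an input action icon beside an uppercase label.
# On Construct, it reads the current CommonInput type and scales the icon's ScaleBox to full size for gamepad, smaller otherwise.
# A SizeBox caps label width, and hover and press events play 2D sounds from variables.
# Mobile and PC/console button style variables are declared but unused in these graphs.
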WidgetBlueprint /Game/UI/Frontend/BottomBar/AthenaHUDBoundActionButton
graphs [
 EventGraph {nodes [
   Event Construct()
   > UpdateInputActionIconSize()
   Event On Pressed() {is=title K2Node_Event}
   > PlaySound2D(Sound=Get OnPressedSound() -> (OnPressedSound:Sound Base Object Reference))
   Event On Hovered()
   > PlaySound2D(Sound=Get OnHoveredSound() -> (OnHoveredSound:Sound Base Object Reference))
  ]}
 UpdateInputActionIconSize {nodes [
   UpdateInputActionIconSize() {is=title K2Node_FunctionEntry}
   n1 = GetCurrentInputType(self=Get CommonInputSubsystem!() -> (ReturnValue:Common Input Object Reference) {is=title K2Node_GetBlueprintContextEx})
   > SetUserSpecifiedScale!(
    execute
    self:Scale Box Object Reference=Get InputActionIconScale() -> (InputActionIconScale:Scale Box Object Reference)
    InUserSpecifiedScale:float=Select(Option 0:double=0.75, Option 1:double=1, Index="Equal (Enum)"(A=@n1, B=Gamepad)) -> (ReturnValue:double)
   ) -> (then)
  ]}
]
variables [
 MobileButtonStyle {
  type=Class
  default=/FortUILibrary/Foundation/Buttons/ButtonStyles/TopBar/ButtonStyle-OutlineEmptyFill_Mobile.ButtonStyle-OutlineEmptyFill_Mobile_C
 }
 OnHoveredSound {type=SoundBase, default=/Game/Sounds/Fort_UI_Sounds/FrontEnd/DefaultSounds/PressedButtons/MSS_UI_Button_Hover_Default}
 OnPressedSound {type=SoundBase, default=/Game/Sounds/Fort_UI_Sounds/FrontEnd/DefaultSounds/PressedButtons/MSS_UI_Button_Pressed_Default}
 PCandConsoleButtonStyle {type=Class, default=/FortUILibrary/Foundation/Buttons/ButtonStyles/TopBar/ButtonStyle-OutlineEmptyFill.ButtonStyle-OutlineEmptyFill_C}
]
widgets {
 info {inheritedWidgetCount=2, parentClass=/Script/FortniteUI.FortBoundActionButton, rootWidgetClass=&Border, widgetCount=6}
 widgets [
  ContentBorder {
   widgetClassPath=&Border
   bIsVariable=true
   widget=ContentBorder
   properties {
    background {
     bIsDynamicallyLoaded=false
     drawAs=NoDrawType
     imageType=NoImage
     mirroring=NoMirror
     resourceName=None
     resourceObject=null
     tiling=NoTile
     imageSize=<32 32>
     margin=<ltrb 0 0 0 0>
     outlineSettings {
      bUseBrushTransparency=false
      roundingType=HalfHeightRadius
      width=0
      color {colorUseRule=UseColor_Specified, specifiedColor=<rgba 0 0 0 0>}
      cornerRadii=<0 0 0 0>
     }
     tintColor {colorUseRule=UseColor_Specified, specifiedColor=<rgba 1 1 1 1>}
     uVRegion {bIsValid=false, max=<0 0>, min=<0 0>}
    }
    padding=<ltrb 15 0 15 0>
   }
  }
  HorizontalBox {
   widgetClassPath=&HorizontalBox
   parent=ContentBorder
   widget=HorizontalBox
   properties {flowDirectionPreference=Culture}
   slot {class=BorderSlot, properties {padding=<ltrb 15 0 15 0>}}
  }
  InputActionIconScale {
   widgetClassPath=&ScaleBox
   bIsVariable=true
   parent=HorizontalBox
   widget=InputActionIconScale
   properties {stretch=UserSpecified}
   slot {class=HorizontalBoxSlot, properties {horizontalAlignment=HAlign_Center, verticalAlignment=VAlign_Center, padding=<ltrb 0 -16 8 -16>}}
  }
  InputActionWidget {
   widgetClassPath=&CommonActionWidget
   bInherited=true
   bIsVariable=true
   parent=InputActionIconScale
   widget=InputActionWidget
   properties {
    progressMaterialParam=Percentage
    visibility=Collapsed
    progressMaterialBrush {
     bIsDynamicallyLoaded=false
     drawAs=Image
     imageType=NoImage
     mirroring=NoMirror
     resourceName=None
     resourceObject=/Game/UI/Materials/HollowCircularProgressBar
     tiling=NoTile
     imageSize=<72 72>
     margin=<ltrb 0 0 0 0>
     outlineSettings {
      bUseBrushTransparency=false
      roundingType=HalfHeightRadius
      width=0
      color {colorUseRule=UseColor_Specified, specifiedColor=<rgba 0 0 0 0>}
      cornerRadii=<0 0 0 0>
     }
     tintColor {colorUseRule=UseColor_Specified, specifiedColor=<rgba 1 1 0 1>}
     uVRegion {bIsValid=false, max=<0 0>, min=<0 0>}
    }
   }
   slot {class=ScaleBoxSlot}
  }
  MaxwidthForMobileSaver {
   widgetClassPath=&SizeBox
   parent=HorizontalBox
   widget=MaxwidthForMobileSaver
   properties {bOverride_MaxDesiredWidth=true, maxDesiredWidth=310}
   slot {class=HorizontalBoxSlot, properties {verticalAlignment=VAlign_Center, padding=<ltrb 0 0 8 0>, size {sizeRule=Fill, value=1}}}
  }
  Text_ActionName {
   widgetClassPath=&CommonTextBlock
   bInherited=true
   parent=MaxwidthForMobileSaver
   widget=Text_ActionName
   properties {
    justification=Center
    mobileFontSizeMultiplier=1.3
    scrollStyle=&TextScrollStyle-Base_C
    style=/FortUILibrary/Foundation/Buttons/TextStyles/TextStyle-Button-Primary-M.TextStyle-Button-Primary-M_C
    text=Text Block
    textTransformPolicy=ToUpper
    colorAndOpacity {colorUseRule=UseColor_Specified, specifiedColor=<rgba 0.008 0.011 0.0262 1>}
    font {
     bForceMonospaced=false
     bMaterialIsStencil=false
     fontMaterial=null
     fontObject=&NotoSans
     letterSpacing=0
     monospacedWidth=1
     size=18
     skewAmount=0
     typefaceFontName=NotoSans-Bold
     outlineSettings {
      bApplyOutlineToDropShadows=false
      bMiteredCorners=false
      bSeparateFillAlpha=false
      outlineMaterial=null
      outlineSize=0
      outlineColor=<rgba 0 0 0 1>
     }
    }
    shadowOffset=<0 0>
   }
   slot {class=SizeBoxSlot}
  }
 ]
}
