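# Frontend action bar container widget that places a backplate behind a named slot for button content.
# The backplate shows, without blocking hit tests, only when the current input is gamepad and HasPromptsToDisplay is true; otherwise it collapses.
# Visibility refreshes on construct, SetHasPrompts calls and CommonInput method changes, with the delegate unbound on destruct.
# BackplateMaterial is applied on initialization and can be swapped through Change Backplate Material.
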WidgetBlueprint /Game/UI/Frontend/BottomBar/WBP_ActionBarContainer
graphs [
 EventGraph {nodes [
   Handle Input Method Changed!() -> (OutputDelegate, then, New Input Type:ECommonInputType Enum) {is=title K2Node_CustomEvent}
   > UpdateBackplateVisibility()
   n1 = Change Backplate Material!() -> (OutputDelegate, then, New Backplate Material:Material Instance Object Reference) {is=title K2Node_CustomEvent}
   > n2 = Set BackplateMaterial(BackplateMaterial:Material Instance Object Reference=@n1.New Backplate Material) -> (Output_Get:Material Instance Object Reference)
   > SetBackplateMaterial!(
    execute
    self:WBP UIKit Backplate Base Object Reference=Get GamepadBackplate() -> (GamepadBackplate:WBP UIKit Backplate Object Reference)
    Material Instance:Material Instance Object Reference=@n2
   ) -> (then)
   Event OnInitialized()
   > SetBackplateMaterial(self=Get GamepadBackplate(), Material Instance=Get BackplateMaterial() -> (BackplateMaterial:Material Instance Object Reference))
   Event Construct()
   n3 = Get CommonInputSubsystem()
   > Is Valid(InputObject=@n3)
   n4 = Create Event()
   > Bind Event to On Input Method Changed(self=@n3, Delegate=@n4)
   > UpdateBackplateVisibility()
   Event Destruct()
   > Is Valid(InputObject=@n3)
   > Unbind Event from On Input Method Changed(self=@n3, Delegate=@n4)
  ]}
 SetHasPrompts {nodes [
   n1 = SetHasPrompts!() -> (then, HasPrompts:bool=~) {is=title K2Node_FunctionEntry}
   > Set HasPromptsToDisplay(HasPromptsToDisplay:bool=@n1) -> (Output_Get:bool)
   > UpdateBackplateVisibility()
  ]}
 UpdateBackplateVisibility {nodes [
   UpdateBackplateVisibility() {is=title K2Node_FunctionEntry}
   n1 = Get CommonInputSubsystem()
   > Is Valid(InputObject=@n1)
   n2 = "Equal (Enum)"(A=GetCurrentInputType(self=@n1), B=Gamepad)
   n3 = Get HasPromptsToDisplay() -> (HasPromptsToDisplay:bool)
   > SetVisibility(self=Get GamepadBackplate(), InVisibility=Select(Option 0=Collapsed, Option 1=HitTestInvisible, Index=AND Boolean(A=@n3, B=@n2)))
  ]}
]
variables [
 BackplateMaterial {type=MaterialInstance, default=/FortUILibrary/Systems/Backplate/MaterialInstances/MI_UIKit_Backplate}
 HasPromptsToDisplay {type=boolean, default=true}
]
widgets {
 info {inheritedWidgetCount=0, parentClass=&UserWidget, rootWidgetClass=&Overlay, widgetCount=3}
 widgets [
  Overlay_32 {widgetClassPath=&Overlay, widget=Overlay_32}
  GamepadBackplate {
   widgetClassPath=&WBP_UIKit_Backplate
   bIsVariable=true
   parent=Overlay_32
   widget=GamepadBackplate
   properties {brightness=Obscure, cornerRadius=Circle}
   slot {class=OverlaySlot, properties {horizontalAlignment=HAlign_Fill, verticalAlignment=VAlign_Fill}}
  }
  ButtonsContentSlot {
   widgetClassPath=&NamedSlot
   bIsVariable=true
   parent=Overlay_32
   widget=ButtonsContentSlot
   slot {class=OverlaySlot, properties {horizontalAlignment=HAlign_Fill, verticalAlignment=VAlign_Fill}}
  }
 ]
}
