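# Athena frontend bottom bar widget that wraps a bound action bar and adds chat notification count and mic state to its buttons.
# Gameplay event router listeners cache both values, with the mic listener skipped when touch HUD mic buttons are supported.
# Value changes and action bar updates push them into every quiet bottom-bar button.
# Bar updates also broadcast prompt presence and collapse the widget when empty, while input-method changes adjust button spacing for gamepad.
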WidgetBlueprint /Game/UI/Frontend/BottomBar/WBP_AthenaBottomBar_WithNotifications
graphs [
 EventGraph {nodes [
   HandleInputMethodChanged!() -> (OutputDelegate, then, NewInputType:ECommonInputType Enum) {is=title K2Node_CustomEvent}
   > UpdateButtonSpacing()
   OnActionBarUpdated() {is=title K2Node_CustomEvent}
   n1 = "integer > integer"(A=GetNumEntries!(self:Dynamic Entry Box Base Object Reference=Get AthenaBoundActionBar() -> (AthenaBoundActionBar:Athena Bound Action Bar Bottom Bar Object Reference)) -> (ReturnValue:int=0))
   > Call Set Has Prompts!(execute:Exec, self:WBP Athena Bottom Bar With Notifications Object Reference, HasPrompts:bool=@n1) -> (then:Exec) {is=title K2Node_CallDelegate}
   > SetVisibility(InVisibility=Select(Option 0=Collapsed, Option 1=SelfHitTestInvisible, Index=@n1))
   > SetUniqueButtons()
   Event OnInitialized()
   > Parent: OnInitialized() {is=title K2Node_CallParentFunction}
   > SetActionBarWidget!(execute, self, NewActionBarWidget:Athena Bound Action Bar Object Reference=Get AthenaBoundActionBar()) -> (then)
   Event Construct()
   n2 = Create Event()
   n3 = Get CommonInputSubsystem()
   > Bind Event to On Input Method Changed(self=@n3, Delegate=@n2)
   > UpdateButtonSpacing()
   > n4 = Sequence()
   n5 = GetOwningPlayer()
   n6 = GetGlobalEventRouter()
   n7 = StartListeningToEvent!(
    execute:Exec=@n4.then_1
    Target:Gameplay Event Router Component Object Reference=@n6
    EventType:Script Struct Object Reference=&/BPS_BottomBarNotificationCount
    Context:Object Reference=@n5
    EventBubblingRule:EEventBubblingRule Enum
   ) -> (
    then:Exec
    AsyncTaskProxy:Async Action Start Listening To Event Object Reference
    OnEventReceived:Exec
    Payload:BPS Bottom Bar Notification Count Structure
   )
   > Set ChatNotificationEventReference(ChatNotificationEventReference:Async Action Start Listening To Event Object Reference=@n7) -> (Output_Get:Async Action Start Listening To Event Object Reference)
   Set CurrentChatNotificationCount(
    execute=@n7.OnEventReceived
    CurrentChatNotificationCount:int=Break BPS Bottom Bar Notification Count!(BPS_BottomBarNotificationCount:"BPS Bottom Bar Notification Count Structure (by ref)"=@n7.Payload) -> (NotificationCount_8_#1:int) {is=title K2Node_BreakStruct}
   ) -> (Output_Get:int=0)
   > SetUniqueButtons()
   n8 = Branch(
    execute=@n4.then_0
    Condition=Get bSupportsTouchHudMicButton!(self:Fort Runtime Options Object Reference=GetRuntimeOptions()) -> (bSupportsTouchHudMicButton:bool)
   )
   n9 = StartListeningToEvent(execute=@n8.else, Target=@n6, EventType=&/BPS_BottomBarMicToggleStateChanged, Context=@n5) -> (Payload:BPS Bottom Bar Mic Toggle State Changed Structure)
   > Set MobileMicToggleEventReference(MobileMicToggleEventReference:Async Action Start Listening To Event Object Reference=@n9) -> (Output_Get:Async Action Start Listening To Event Object Reference)
   Set IsMicOn(
    execute=@n9.OnEventReceived
    IsMicOn:bool=Break BPS Bottom Bar Mic Toggle State Changed!(BPS_BottomBarMicToggleStateChanged:"BPS Bottom Bar Mic Toggle State Changed Structure (by ref)"=@n9.Payload) -> (IsMicOn_2_#2:bool=~) {is=title K2Node_BreakStruct}
   ) -> (Output_Get:bool)
   > SetUniqueButtons()
   Event Destruct()
   > Unbind Event from On Input Method Changed(self=@n3, Delegate=@n2)
   > StopListeningToEvent(self:Async Action Start Listening To Event Object Reference=Get ChatNotificationEventReference() -> (ChatNotificationEventReference:Async Action Start Listening To Event Object Reference))
   > StopListeningToEvent(self=Get MobileMicToggleEventReference() -> (MobileMicToggleEventReference:Async Action Start Listening To Event Object Reference))
  ]}
 OnSelfIsMutedChanged {nodes [
   n1 = OnSelfIsMutedChanged!() -> (then, bIsMuted:bool=~) {is=title K2Node_FunctionEntry}
   > Set IsMicOn(IsMicOn=NOT Boolean(A=@n1))
   > SetUniqueButtons()
  ]}
 SetActionBarWidget {nodes [
   SetActionBarWidget!() -> (then, NewActionBarWidget:Athena Bound Action Bar Object Reference) {is=title K2Node_FunctionEntry}
   n1 = Create Event()
   > Unbind Event from On Action Bar Updated(self:Common Bound Action Bar Object Reference=Get AthenaBoundActionBar(), Delegate=@n1) {is=title K2Node_RemoveDelegate}
   > Bind Event to On Action Bar Updated(self:Common Bound Action Bar Object Reference=Get AthenaBoundActionBar(), Delegate=@n1) {is=title K2Node_AddDelegate}
  ]}
 SetHasPrompts {nodes [SetHasPrompts!() -> (then, HasPrompts:bool=~) {is=title K2Node_FunctionEntry}]}
 SetUniqueButtons {nodes [
   SetUniqueButtons() {is=title K2Node_FunctionEntry}
   > n1 = For Each Loop(Array:Array of User Widget Object References=GetAllEntries!(self:Dynamic Entry Box Base Object Reference=Get AthenaBoundActionBar()) -> (ReturnValue:Array of User Widget Object References)) -> (Array Element:User Widget Object Reference)
   > n2 = Cast To WBP_UIKit_Button_Quiet_BottomBar!(execute, Object=@n1) -> (then, CastFailed, AsWBP UIKit Button Quiet Bottom Bar:WBP UIKit Button Quiet Bottom Bar Object Reference) {is=title K2Node_DynamicCast}
   > SetUniqueButtonAttributes!(
    execute
    self:WBP UIKit Button Quiet Bottom Bar Object Reference=@n2
    NotificationCount:int=Get CurrentChatNotificationCount() -> (CurrentChatNotificationCount:int=0)
    IsMicOn:bool=Get IsMicOn() -> (IsMicOn:bool)
   ) -> (then)
  ]}
 UpdateButtonSpacing {nodes [
   UpdateButtonSpacing() {is=title K2Node_FunctionEntry}
   n1 = GetCurrentInputType(self=Get CommonInputSubsystem())
   > SetEntrySpacing!(
    execute
    self:Dynamic Entry Box Base Object Reference=Get AthenaBoundActionBar()
    InEntrySpacing:"Vector 2D Structure (by ref)"=Select(
     Option 0:"Vector 2D Structure (by ref)"=MakeVector2D(X=24, Y=0)
     Option 1:"Vector 2D Structure (by ref)"=MakeVector2D(X=0, Y=0)
     Index="Equal (Enum)"(A=@n1, B=Gamepad)
    ) -> (ReturnValue:"Vector 2D Structure (by ref)")
   ) -> (then)
  ]}
]
variables [
 ActionBarWidget {type=AthenaBoundActionBar, default=null}
 ChatNotificationEventReference {type=AsyncAction_StartListeningToEvent, default=null}
 CurrentChatNotificationCount {type=integer}
 IsMicOn {type=boolean}
 MobileMicToggleEventReference {type=AsyncAction_StartListeningToEvent, default=null}
 SetHasPrompts {}
]
widgets {
 info {
  inheritedWidgetCount=0
  parentClass=&CommonUserWidget
  rootWidgetClass=&/AthenaBoundActionBar_BottomBar.AthenaBoundActionBar_BottomBar_C
  widgetCount=1
 }
 widgets [AthenaBoundActionBar {widgetClassPath=&/AthenaBoundActionBar_BottomBar, bIsVariable=true, widget=AthenaBoundActionBar}]
}
