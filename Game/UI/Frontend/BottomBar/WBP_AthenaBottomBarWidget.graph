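# Frontend bottom bar widget that hosts an action bar in a safe zone, with mobile padding and a backplate material chosen by game mode.
# Slot_ActionBar holds a notifications bottom bar by default, or a supplied AthenaBoundActionBar.
# Prompt updates from the active bar tell WBP_ActionBarContainer whether it has prompts.
# For a supplied bar, delta-gated bindings track entry updates and input method changes, which switch button spacing between gamepad and other input.
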WidgetBlueprint /Game/UI/Frontend/BottomBar/WBP_AthenaBottomBarWidget
graphs [
 BindToInputChanged {nodes [
   n1 = BindToInputChanged!() -> (then, DoBind:bool=~) {is=title K2Node_FunctionEntry}
   > n2 = Delta Gate Wildcard!(execute:Exec, Variable:"Boolean (by ref)"=Get IsInputChangedBound() -> (IsInputChangedBound:bool), New Value:bool=@n1) -> (Success:Exec, Updated Variable:"Boolean (by ref)", Previous Variable:bool=~) {is=title K2Node_MacroInstance}
   > n3 = Branch(Condition=@n2)
   n4 = Create Event()
   n5 = Get CommonInputSubsystem()
   > Bind Event to On Input Method Changed(self=@n5, Delegate=@n4)
   > UpdateButtonSpacing()
   Unbind Event from On Input Method Changed(execute=@n3.else, self=@n5, Delegate=@n4)
  ]}
 EventGraph {nodes [
   HandleInputMethodChanged!() -> (OutputDelegate, then, NewInputType:ECommonInputType Enum) {is=title K2Node_CustomEvent}
   > UpdateButtonSpacing()
   OnActionBarUpdated() {is=title K2Node_CustomEvent}
   n1 = Get ActionBarWidget() -> (ActionBarWidget:Athena Bound Action Bar Object Reference)
   > OnUpdateHasPrompts!(
    execute
    self
    HasPrompts:bool="integer > integer"(A=GetNumEntries!(self:Dynamic Entry Box Base Object Reference=@n1) -> (ReturnValue:int=0))
   ) -> (then)
   n2 = OnUpdateHasPrompts!() -> (OutputDelegate, then, HasPrompts:bool=~) {is=title K2Node_CustomEvent}
   > SetHasPrompts!(
    execute
    self:WBP Action Bar Container Object Reference=Get WBP_ActionBarContainer() -> (WBP_ActionBarContainer:WBP Action Bar Container Object Reference)
    HasPrompts:bool=@n2.HasPrompts
   ) -> (then)
   Event OnInitialized()
   > Parent: OnInitialized() {is=title K2Node_CallParentFunction}
   > Branch(Condition=IsMobileGame())
   > SetPadding!(
    execute
    self:Overlay Slot Object Reference=SlotAsOverlaySlot(Widget=Get WBP_ActionBarContainer())
    InPadding_Left:float=0
    InPadding_Top:float=0
    InPadding_Right:float=48
    InPadding_Bottom:float=24
   ) -> (then)
   Event PreConstruct()
   > n3 = Cast To FortGameModeFrontEnd!(execute, Object=GetGameMode!() -> (ReturnValue:Game Mode Base Object Reference)) -> (then, CastFailed, AsFort Game Mode Front End:Fort Game Mode Front End Object Reference) {is=title K2Node_DynamicCast}
   n4 = Get WBP_ActionBarContainer()
   > Change Backplate Material!(
    execute
    self:WBP Action Bar Container Object Reference=@n4
    New Backplate Material:Material Instance Object Reference=/FortUILibrary/Systems/Backplate/MaterialInstances/MI_UIKit_Backplate
   ) -> (then)
   Change Backplate Material(execute=@n3.CastFailed, self=@n4, New Backplate Material=/FortUILibrary/Systems/Backplate/MaterialInstances/MI_UIKit_Backplate_Fallback)
   Event Construct()
   > n5 = Sequence()
   n6 = GetChildAt!(self:Panel Widget Object Reference=Get Slot_ActionBar() -> (Slot_ActionBar:Named Slot Object Reference), Index:int=0) -> (ReturnValue:Widget Object Reference)
   > Is Valid(InputObject=@n6)
   > n7 = Cast To AthenaBoundActionBar!(execute, Object=@n6) -> (then, CastFailed, AsAthena Bound Action Bar:Athena Bound Action Bar Object Reference) {is=title K2Node_DynamicCast}
   > SetActionBarWidget!(execute, self, NewActionBarWidget:Athena Bound Action Bar Object Reference=@n7) -> (then)
   n8 = Branch(execute=@n5.then_1, Condition=Get IsActionBarSet() -> (IsActionBarSet:bool))
   SetDisplayOwningPlayerActionsOnly!(
    execute=@n8.else
    self:Common Bound Action Bar Object Reference=Get AthenaBoundActionBar!(self:WBP Athena Bottom Bar With Notifications Object Reference=Get WBP_AthenaBottomBar_WithNotifications() -> (WBP_AthenaBottomBar_WithNotifications:WBP Athena Bottom Bar With Notifications Object Reference)) -> (AthenaBoundActionBar:Athena Bound Action Bar Bottom Bar Object Reference)
    bShouldOnlyDisplayOwningPlayerActions:bool=Get Display Owner Actions Only() -> (Display Owner Actions Only:bool)
   ) -> (then)
   > Bind Event to Set Has Prompts(self:WBP Athena Bottom Bar With Notifications Object Reference=Get WBP_AthenaBottomBar_WithNotifications(), Delegate=Create Event()) {is=title K2Node_AddDelegate}
   Event Destruct()
   > n9 = Sequence()
   > n10 = Branch(Condition=Get IsActionBarSet())
   Unbind Event from Set Has Prompts(
    execute=@n10.else
    self:WBP Athena Bottom Bar With Notifications Object Reference=Get WBP_AthenaBottomBar_WithNotifications()
    Delegate=Create Event()
   ) {is=title K2Node_RemoveDelegate}
   SetActionBarWidget(execute=@n9.then_1)
   > BindToInputChanged!(execute, self, DoBind:bool=false) -> (then) {is=title K2Node_CallFunction}
  ]}
 SetActionBarWidget {nodes [
   n1 = SetActionBarWidget!() -> (then, NewActionBarWidget:Athena Bound Action Bar Object Reference) {is=title K2Node_FunctionEntry}
   > n2 = Delta Gate Wildcard(Variable:"Athena Bound Action Bar Object Reference (by ref)"=Get ActionBarWidget(), New Value:Athena Bound Action Bar Object Reference=@n1) -> (Updated Variable:"Athena Bound Action Bar Object Reference (by ref)", Previous Variable:Athena Bound Action Bar Object Reference)
   > n3 = Is Valid(InputObject=@n2.Previous Variable)
   > Unbind Event from On Action Bar Updated(self:Common Bound Action Bar Object Reference=@n2.Previous Variable, Delegate=Create Event()) {is=title K2Node_RemoveDelegate}
   > n4 = Set IsActionBarSet(IsActionBarSet:bool=false) -> (Output_Get:bool)
   n5 = Get!(execute:Exec=[@n4.then, @n3.Is Not Valid]) -> (then:Exec, else:Exec, ActionBarWidget:Athena Bound Action Bar Object Reference)
   > Set IsActionBarSet(IsActionBarSet=true)
   > Bind Event to On Action Bar Updated(self:Common Bound Action Bar Object Reference=@n5, Delegate=Create Event()) {is=title K2Node_AddDelegate}
   > BindToInputChanged(DoBind=true)
   > SetDisplayOwningPlayerActionsOnly(self=Get ActionBarWidget(), bShouldOnlyDisplayOwningPlayerActions=Get Display Owner Actions Only())
  ]}
 UpdateButtonSpacing {nodes [
   UpdateButtonSpacing() {is=title K2Node_FunctionEntry}
   n1 = GetCurrentInputType(self=Get CommonInputSubsystem())
   > SetEntrySpacing!(
    execute
    self:Dynamic Entry Box Base Object Reference=Get ActionBarWidget()
    InEntrySpacing:"Vector 2D Structure (by ref)"=Select(
     Option 0:"Vector 2D Structure (by ref)"=MakeVector2D(X=24, Y=0)
     Option 1:"Vector 2D Structure (by ref)"=MakeVector2D(X=0, Y=0)
     Index="Equal (Enum)"(A=@n1, B=Gamepad)
    ) -> (ReturnValue:"Vector 2D Structure (by ref)")
   ) -> (then)
  ]}
]
variables [
 ActionBarWidget {type=AthenaBoundActionBar, default=null}
 Display Owner Actions Only {type=boolean, default=true}
 IsActionBarSet {type=boolean}
 IsInputChangedBound {type=boolean}
]
widgets {
 info {
  inheritedWidgetCount=0
  namedSlotCount=2
  parentClass=&CommonUserWidget
  rootWidgetClass=/Game/Athena/HUD/FortHUDElementWrapper.FortHUDElementWrapper_C
  widgetCount=7
 }
 widgets [
  FortHUDElementWrapper {
   widgetClassPath=/Game/Athena/HUD/FortHUDElementWrapper
   bIsVariable=true
   widget=FortHUDElementWrapper
   properties {elementTag {tagName=HUD}}
  }
  SafeZone_BottomBar {
   widgetClassPath=&SafeZone
   namedSlotHost=FortHUDElementWrapper
   widget=SafeZone_BottomBar
   properties {padLeft=false, padTop=false}
  }
  Overlay_0 {widgetClassPath=&Overlay, parent=SafeZone_BottomBar, widget=Overlay_0, slot {class=SafeZoneSlot, properties {bIsTitleSafe=false}}}
  BufferSlot {
   widgetClassPath=&NamedSlot
   bIsVariable=true
   parent=Overlay_0
   widget=BufferSlot
   properties {bExposeOnInstanceOnly=true}
   slot {class=OverlaySlot, properties {horizontalAlignment=HAlign_Fill, verticalAlignment=VAlign_Fill}}
  }
  WBP_ActionBarContainer {
   widgetClassPath=&/WBP_ActionBarContainer
   bIsVariable=true
   parent=Overlay_0
   widget=WBP_ActionBarContainer
   slot {class=OverlaySlot, properties {horizontalAlignment=HAlign_Fill, verticalAlignment=VAlign_Fill, padding=<ltrb 0 0 80 32>}}
  }
  Slot_ActionBar {
   widgetClassPath=&NamedSlot
   bIsVariable=true
   namedSlotHost=WBP_ActionBarContainer
   widget=Slot_ActionBar
   properties {bExposeOnInstanceOnly=true}
  }
  WBP_AthenaBottomBar_WithNotifications {
   widgetClassPath=&/WBP_AthenaBottomBar_WithNotifications
   bIsVariable=true
   parent=Slot_ActionBar
   widget=WBP_AthenaBottomBar_WithNotifications
   slot {class=PanelSlot}
  }
 ]
}
