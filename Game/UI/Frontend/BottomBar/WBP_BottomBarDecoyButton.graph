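# Widget-less CommonButtonBase decoy for the frontend bottom bar, carrying triggering input actions without appearing on screen.
# On construct it stays Visible but is scaled to zero width.
# PreConstruct caches its triggering data-table and Enhanced Input actions once.
# SetBottomButtonVisibility toggles the bar prompt by swapping between those cached actions and empty ones, rather than changing widget visibility.
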WidgetBlueprint /Game/UI/Frontend/BottomBar/WBP_BottomBarDecoyButton
graphs [
 CacheInputAction {nodes [
   CacheInputAction() {is=title K2Node_FunctionEntry}
   > Set BottomButtonInputAction(BottomButtonInputAction:Data Table Row Handle Structure=Get TriggeringInputAction() -> (TriggeringInputAction:Data Table Row Handle Structure)) -> (Output_Get:Data Table Row Handle Structure)
   > Set BottomButtonEnhancedInputAction(BottomButtonEnhancedInputAction:Input Action Object Reference=Get TriggeringEnhancedInputAction() -> (TriggeringEnhancedInputAction:Input Action Object Reference)) -> (Output_Get:Input Action Object Reference)
  ]}
 EventGraph {nodes [
   Event Construct()
   > Parent: Construct()
   > SetRenderScale!(execute, self:Widget Object Reference, Scale:Vector 2D Structure=(X=0,Y=1)) -> (then)
   > SetVisibility(InVisibility=Visible)
   n1 = Event PreConstruct()
   > Parent: PreConstruct(IsDesignTime=@n1.IsDesignTime)
   > n2 = Sequence()
   > Do Once()
   > CacheInputAction() {is=title K2Node_CallFunction}
   Branch(execute=@n2.then_1, Condition=@n1.IsDesignTime)
  ]}
 SetBottomButtonVisibility {nodes [
   n1 = SetBottomButtonVisibility!() -> (then, ShouldButtonBeVisible:bool=~) {is=title K2Node_FunctionEntry}
   > SetTriggeringInputAction!(
    execute
    self:Common Button Base Object Reference
    InputActionRow:"Data Table Row Handle Structure (by ref)"=Select(
     Option 0:"Data Table Row Handle Structure (by ref)"=Get No Input Action Row() -> (No Input Action Row:Data Table Row Handle Structure)
     Option 1:"Data Table Row Handle Structure (by ref)"=Get BottomButtonInputAction() -> (BottomButtonInputAction:Data Table Row Handle Structure)
     Index=@n1
    ) -> (ReturnValue:"Data Table Row Handle Structure (by ref)")
   ) -> (then)
   > SetTriggeringEnhancedInputAction!(
    execute
    self:Common Button Base Object Reference
    InInputAction:Input Action Object Reference=Select(
     Option 0:Input Action Object Reference=Get NoEnhancedInputAction() -> (NoEnhancedInputAction:Input Action Object Reference)
     Option 1:Input Action Object Reference=Get BottomButtonEnhancedInputAction() -> (BottomButtonEnhancedInputAction:Input Action Object Reference)
     Index=@n1
    ) -> (ReturnValue:Input Action Object Reference)
   ) -> (then)
  ]}
]
variables [
 BottomButtonEnhancedInputAction {type=InputAction, default=null}
 BottomButtonInputAction {type=DataTableRowHandle, default {dataTable=&InputActionDataTable, rowName=Back}}
 No Input Action Row {type=DataTableRowHandle, default {dataTable=null, rowName=None}}
 NoEnhancedInputAction {type=InputAction, default=null}
]
widgets {info {inheritedWidgetCount=0, parentClass=&CommonButtonBase, rootWidgetClass=null, widgetCount=0}, widgets []}
